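# Material function for the Season 2 sea mist lighthouse that supplies a pulsing intensity value and an on/off state.
# Time divided by the "Wave Lenght" parameter drives a sine, which is offset and halved into a smooth 0–1 "Intensity Sine" output.
# The "LightHouse On/Off State" parameter is saturated into a separate "On/Off" output and does not modulate the wave inside the function.
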MaterialFunction /Hera/TechArt/Season2/SeaMist/MF_Lighthouse_IntensityWave
graph {nodes [
  n1 = Time()
  n2 = ScalarParameter() {defaultValue=5, parameterName=Wave Lenght}
  n3 = Add(A=Sine(None=Divide(A=@n1, B=@n2) {constB=5}))
  FunctionOutput(None=NamedRerouteDeclaration(None=Multiply(A=@n3) {constB=0.5}) {name=Sine Wave, nodeColor=<rgba 1 0 0.2812 1>}) {outputName=Intensity Sine}
  FunctionOutput(None=Saturate(None=ScalarParameter() {defaultValue=1, parameterName=LightHouse On/Off State})) {outputName=On/Off}
 ]}
