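# Light function material for the lighthouse light in the Season 2 SeaMist content, flagged as compatible with the light function atlas.
# MF_Lighthouse_IntensityWave supplies an On/Off signal and an intensity sine.
# The inverted sine is multiplied by On/Off, forming a switched, pulsing intensity term.
# A LinearInterpolate node has no connected inputs, and no connection to the material output is shown.
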Material /Hera/TechArt/Season2/SeaMist/LF_LightHouse_Light
graph {
 properties {
  bForceCompatibleWithLightFunctionAtlas=true
  materialDomain=MD_LightFunction
  pixelDepthOffsetMode=PDOM_Legacy
  usedShadingModels=Default Lit
 }
 nodes [
  LinearInterpolate() {constA=1, constB=0}
  n1 = MaterialFunctionCall() {materialFunction=&/MF_Lighthouse_IntensityWave}
  Multiply(A=@n1.On/Off, B=OneMinus(None=@n1.Intensity Sine))
 ]
}
