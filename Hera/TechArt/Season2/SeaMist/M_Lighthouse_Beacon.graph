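# Opaque material for the Sea Mist lighthouse Fresnel lens beacon, with tinted glass and a pulsing, switchable glow.
# Texture masks and view-facing normal terms shape lens highlight and emission, kept exposure-stable with EyeAdaptationInverse and modulated by MF_Lighthouse_IntensityWave.
# Bump-offset, blurred normals fake a time-of-day-tinted interior, zeroed on low and medium quality, while pixel depth builds a distance mask.
# Proxy rendering gets a simpler base color before destructible damage is applied.
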Material /Hera/TechArt/Season2/SeaMist/M_Lighthouse_Beacon
graph {
 properties {
  bAutomaticallySetUsageInEditor=false
  bUseMaterialAttributes=true
  bUsedWithNanite=true
  previewMesh=/Hera/Sets/SeaMist/Meshes/SM_SeaMist_Lighthouse_FresnelLens_A
  usedShadingModels=Default Lit
 }
 nodes [
  n1 = Add(A=TextureCoordinate(), B=NamedRerouteUsage())
  n2 = PixelNormalWS()
  n3 = Transform(None=@n2) {transformSourceType=TRANSFORMSOURCE_World, transformType=TRANSFORM_Tangent}
  Add(
   A=@n1
   B=NamedRerouteDeclaration(None=ComponentMask(None=@n3) {g=true, r=true}) {name=Reflection Vector Tangent, nodeColor=<rgba 0.4766 1 0 1>}
  )
  n4 = OneMinus(None=NamedRerouteUsage())
  n5 = Multiply(A=@n4, B=@n4)
  n6 = LocalPosition()
  n7 = Constant() {r=-437.307}
  n8 = OneMinus(None=Divide(A=Add(A=@n6.Z, B=@n7), B=ScalarParameter() {defaultValue=232.5126}))
  Add(A=Multiply(A=@n5) {constB=2}, B=@n8)
  Constant3Vector()
  Constant3Vector() {constant=<rgba 0 0 1 0>}
  Constant()
  Constant() {r=2}
  LinearInterpolate(
   B=VectorParameter() {parameterName=Beacon Color, defaultValue=<rgba 0.2552 0.1456 0.0702 1>}.RGB
   Alpha=Clamp(None=NamedRerouteUsage()) {maxDefault=0.8}
  ) {constA=1}
  n9 = Divide(
   A=Add(A=PixelDepth().R, B=ScalarParameter() {defaultValue=-2000, parameterName=Initial transition distance})
   B=ScalarParameter() {defaultValue=70000, parameterName=Distance}
  ) {constB=4000}
  MaterialFunctionCall(In=@n9, Contrast=Constant() {r=0.5}) {materialFunction=&CheapContrast}
  n10 = NamedRerouteUsage()
  n11 = Desaturation(
   None=Multiply(A=NamedRerouteUsage(), B=NamedRerouteUsage())
   Fraction=LinearInterpolate(B=ScalarParameter() {defaultValue=1, parameterName=BaseColor Desaturation}, Alpha=@n10) {constA=0.6}
  )
  n12 = Saturate(None=Subtract(A=Multiply(A=@n11) {constB=3}) {constB=0.1})
  n13 = Multiply(A=@n12, B=@n12)
  n14 = LinearInterpolate(A=@n12, B=@n13)
  n15 = Multiply(
   A=LinearInterpolate(
    A=VectorParameter() {parameterName=Glass Color off, defaultValue=<rgba 0.6898 0.8021 0.5305 1>}.RGB
    B=VectorParameter() {parameterName=Glass Color, defaultValue=<rgba 1 0.6721 0.3021 1>}.RGB
    Alpha=NamedRerouteUsage()
   )
   B=NamedRerouteDeclaration(None=LinearInterpolate(A=@n14, B=@n13, Alpha=@n10) {constA=0.6}) {name=BaseColor Mask, nodeColor=<rgba 0 1 0.7422 1>}
  )
  n16 = Multiply(A=NamedRerouteUsage(), B=@n15)
  n17 = Transform(None=NamedRerouteUsage())
  n18 = CameraVectorWS()
  n19 = Multiply(
   A=Power(Base=DotProduct(A=@n17, B=@n18), Exp=ScalarParameter() {defaultValue=3, group=Emission, parameterName=Emission Fresnel Power}) {constExponent=8}
   B=NamedRerouteUsage()
  )
  n20 = Constant() {r=0.25}
  n21 = Multiply(
   A=Multiply(
    A=RayTracingQualitySwitch(Normal=@n19, RayTraced=@n20)
    B=VectorParameter() {group=Emission, parameterName=Emission Color, defaultValue=<rgba 1 0.588 0.222 1>}.RGB
   )
   B=ScalarParameter() {defaultValue=50, group=Emission, parameterName=Emission Intensity}
  ) {constB=0}
  n22 = Constant() {r=0.5}
  n23 = MaterialFunctionCall() {materialFunction=&/MF_Lighthouse_IntensityWave}
  n24 = NamedRerouteDeclaration(None=@n23.On/Off) {name=Turned Off, nodeColor=<rgba 0.9219 1 0 1>}
  n25 = LinearInterpolate(Alpha=OneMinus(None=@n23.Intensity Sine)) {constA=0.4}
  n26 = SetMaterialAttributes!(
   MaterialAttributes
   Base Color=LinearInterpolate(A=Add(A=@n15, B=@n16), B=@n15, Alpha=NamedRerouteUsage())
   Metallic=NamedRerouteUsage()
   Specular=NamedRerouteUsage()
   Roughness=NamedRerouteUsage()
   Emissive Color=Multiply(A=EyeAdaptationInverse(LightValueInput=@n21, AlphaInput=@n22).EyeAdaptationInverse, B=Multiply(A=@n24, B=@n25))
   Normal=NamedRerouteUsage()
  ) {attributeSetTypes [
    #1
    #2
    #3
    #4
    #5
    #6
   ]}
  n27 = Multiply(A=Multiply(A=VectorParameter() {parameterName=Glass Color, defaultValue=<rgba 1 0.6721 0.3021 1>}.RGB, B=NamedRerouteUsage())) {constB=2}
  MaterialFunctionCall(Material_In=MaterialProxyReplace(
    Realtime=@n26
    MaterialProxy=SetMaterialAttributes!(MaterialAttributes=@n26, Base Color=@n27) {attributeSetTypes [#1]}
   )) {materialFunction=&MF_FN_DestructibleDamage}
  n28 = Transform(None=NamedRerouteUsage())
  n29 = DotProduct(
   A=LinearInterpolate(A=NamedRerouteDeclaration(None=@n28) {name=World Normal Tex, nodeColor=<rgba 0 1 0.0625 1>}, B=VertexNormalWS()) {constAlpha=0.8}
   B=CameraVectorWS()
  )
  n30 = NamedRerouteUsage()
  n31 = Multiply(
   A=Power(Base=@n29) {constExponent=20}
   B=MaterialFunctionCall(In=OneMinus(None=@n30), Contrast=Constant() {r=1}) {materialFunction=&CheapContrast}.Result
  )
  n32 = LinearInterpolate(
   A=VectorParameter() {parameterName=Beacon Color, defaultValue=<rgba 0.2552 0.1456 0.0702 1>}.RGB
   Alpha=Saturate(None=Multiply(A=@n31) {constB=0.3})
  )
  n33 = NamedRerouteUsage()
  n34 = NamedRerouteUsage()
  n35 = LinearInterpolate(Alpha=NamedRerouteUsage()) {constA=2.5, constB=3}
  n36 = Multiply(A=Add(A=Add(A=@n33, B=@n34)) {constB=0.1}, B=Add(A=Power(Base=@n29, Exp=@n35) {constExponent=3}) {constB=0.02})
  n37 = NamedRerouteUsage()
  n38 = LinearInterpolate(A=@n37, B=Multiply(A=@n37) {constB=400}, Alpha=NamedRerouteUsage())
  n39 = Multiply(
   A=Multiply(A=Saturate(None=@n8), B=Multiply(A=@n32, B=@n36))
   B=LinearInterpolate(A=@n38, B=Multiply(A=@n38) {constB=0.6}, Alpha=NamedRerouteUsage())
  )
  n40 = CollectionParameter() {
   collection=&FortniteMaterialParameters
   parameterName=FortniteActiveTimeOfDay
  }
  n41 = ComponentMask(None=@n40) {g=true}
  Multiply(A=@n39, B=LinearInterpolate(Alpha=Saturate(None=@n41)) {constA=1, constB=0})
  n42 = TextureSampleParameter2D!(UVs, Level, Apply View MipBias) {
   mipValueMode=TMVM_MipLevel
   parameterName=SFX
   samplerType=SAMPLERTYPE_Masks
   texture=/Hera/Sets/SeaMist/Textures/T_SeaMist_Lighthouse_FresnelLens_A_A_MSK
  }
  NamedRerouteDeclaration(None=@n42.R) {name=AO, nodeColor=<rgba 0 1 0.2266 1>}
  NamedRerouteDeclaration(None=@n42.G) {name=Fine Curves, nodeColor=<rgba 0 1 0.2266 1>}
  NamedRerouteDeclaration(None=@n5) {name=Reverse AO, nodeColor=<rgba 1 0 0.1172 1>}
  n43 = Constant() {r=-8}
  NamedRerouteDeclaration(None=TextureSample!(UVs=BumpOffset(Coordinate=@n1, Height=@n43), Tex, Level=Constant() {r=7}, Apply View MipBias) {mipValueMode=TMVM_MipLevel, texture=/Hera/Sets/SeaMist/Textures/T_SeaMist_Lighthouse_FresnelLens_A_A_D}.RGB) {name=Tex Color Distorted, nodeColor=<rgba 1 0 0.6562 1>}
  n44 = NamedRerouteUsage()
  n45 = Add(A=Multiply(A=ComponentMask(None=@n44) {g=true}) {constB=0.1}, B=Multiply(A=NamedRerouteUsage()) {constB=0.2})
  n46 = TextureCoordinate()
  n47 = NamedRerouteUsage()
  n48 = Add(A=Add(A=Add(A=@n45, B=@n46), B=Multiply(A=@n47) {constB=0.1}), B=Constant2Vector() {g=0.25, r=0.2}) {constB=0}
  n49 = Constant2Vector() {g=0.6, r=0.75}
  n50 = TextureSample(UVs=BumpOffset(Coordinate=Multiply(A=@n48, B=@n49), Height=Constant() {r=-2}), Level=Constant() {r=5}) {mipValueMode=TMVM_MipLevel, samplerType=SAMPLERTYPE_Normal, texture=/Hera/Sets/SeaMist/Textures/T_SeaMist_Lighthouse_FresnelLens_A_A_N}
  n51 = OneMinus(None=Fresnel(ExponentIn=ScalarParameter() {defaultValue=-0.2243, parameterName=Fake Interior Fresnel}, Normal=Transform(None=@n50.RGB)) {baseReflectFraction=0})
  n52 = LinearInterpolate(
   A=ScalarParameter() {parameterName=Param_2}
   B=ScalarParameter() {defaultValue=1.7208, parameterName=Param_1}
   Alpha=Saturate(None=Multiply(A=@n51, B=@n51))
  )
  n53 = VectorParameter() {parameterName=3, defaultValue=<rgba 1.5 0.8941 0.6319 1>}
  n54 = MaterialFunctionCall(
   Color_Morning=@n53.RGB
   Color_Day=Constant3Vector() {constant=<rgba 0.4635 0.986 1 1>}
   Color_Evening=@n53.RGB
   Color_Night=VectorParameter() {defaultValue=<rgba 0.9375 1.0608 1.5 1>}.RGB
  ) {materialFunction=&Color_Time_of_Day}
  n55 = Constant()
  NamedRerouteDeclaration(None=QualitySwitch(Default=Multiply(A=@n52, B=@n54.Result), Low=@n55, Medium=@n55)) {name=Fake Interior, nodeColor=<rgba 1 0 0.2578 1>}
  NamedRerouteDeclaration(None=EyeAdaptationInverse(LightValueInput=@n39, AlphaInput=Constant() {r=0.4}).EyeAdaptationInverse) {name=Lens, nodeColor=<rgba 1 0.7031 0 1>}
  NamedRerouteDeclaration(None=@n54.Result) {name=Time of Day color, nodeColor=<rgba 0 1 0.5781 1>}
  n56 = MaterialFunctionCall() {materialFunction=&MF_Lighthouse_IntensityWave}
  NamedRerouteDeclaration(None=@n56.Intensity Sine) {name=Sine Wave, nodeColor=<rgba 1 0 0.2812 1>}
  NamedRerouteDeclaration(None=Saturate(None=@n9)) {name=Distance, nodeColor=<rgba 1 0 0.9844 1>}
  NamedRerouteDeclaration(None=@n42.B) {name=Broad Curves, nodeColor=<rgba 0 1 0.2266 1>}
  n57 = TextureSampleParameter2D!(UVs, Apply View MipBias) {parameterName=Surface, samplerType=SAMPLERTYPE_Masks, texture=/Hera/Sets/SeaMist/Textures/T_SeaMist_Lighthouse_FresnelLens_A_A_S}
  NamedRerouteDeclaration(None=@n57.R) {name=Tex Specular, nodeColor=<rgba 1 0.3047 0 1>}
  NamedRerouteDeclaration(None=@n57.G) {name=Tex Metal, nodeColor=<rgba 1 0.3047 0 1>}
  NamedRerouteDeclaration(None=@n57.B) {name=Tex Rough, nodeColor=<rgba 1 0.3047 0 1>}
  NamedRerouteDeclaration(None=TextureSampleParameter2D() {parameterName=Color, texture=&T_SeaMist_Lighthouse_FresnelLens_A_A_D}.RGB) {name=Tex Color, nodeColor=<rgba 1 0 0.6562 1>}
  n58 = TextureSampleParameter2D() {parameterName=Normal, samplerType=SAMPLERTYPE_Normal, texture=&T_SeaMist_Lighthouse_FresnelLens_A_A_N}
  NamedRerouteDeclaration(None=@n58.RGB) {name=Tex Normal, nodeColor=<rgba 0 0.9766 1 1>}
  NamedRerouteDeclaration(None=LinearInterpolate(B=ScalarParameter() {defaultValue=15, parameterName=Beacon Intensity}, Alpha=NamedRerouteUsage())) {name=Beacon Intensity, nodeColor=<rgba 0.1484 1 0 1>}
  NamedRerouteDeclaration(None=@n56.On/Off) {name=On/Off, nodeColor=<rgba 0 0.5547 1 1>}
  NamedRerouteUsage()
  NamedRerouteUsage()
  NamedRerouteUsage()
  ReflectionVectorWS(CustomWorldNormal=@n2)
  VectorParameter() {parameterName=Glass Color, defaultValue=<rgba 1 0.6721 0.3021 1>}
  VertexInterpolator(VS=VertexNormalWS())
 ]
}
