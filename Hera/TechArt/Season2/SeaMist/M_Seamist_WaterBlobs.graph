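# Translucent, unlit, two-sided material for sea mist water blobs, tinted by time-of-day colors.
# Four scrolling noise samples at lerped depths, offset in time by object position, multiply into a glob mask.
# Screen derivatives build normals; color combines tint with masked fake cubemap reflection, alongside a skylight sample.
# Vertex color shapes depth-faded opacity, while sine and panned noise form vertex-normal displacement.
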Material /Hera/TechArt/Season2/SeaMist/M_Seamist_WaterBlobs
graph {
 properties {
  bAutomaticallySetUsageInEditor=false
  blendMode=BLEND_Translucent
  refractionMethod=RM_IndexOfRefraction
  shadingModel=MSM_Unlit
  translucencyLightingMode=TLM_SurfacePerPixelLighting
  twoSided=true
  usedShadingModels=Unlit
 }
 nodes [
  n1 = CollectionParameter() {
   collection=&FortniteMaterialParameters
   parameterName=SunlightDirectionalVector
  }
  n2 = PixelNormalWS()
  Divide(A=Add(A=DotProduct(A=@n1, B=@n2)))
  n3 = LinearInterpolate(
   A=ScalarParameter() {group=UVs, parameterName=Background Depth Max, sortPriority=3}
   B=ScalarParameter() {defaultValue=0.2, group=UVs, parameterName=Background Depth Min, sortPriority=2}
  ) {constAlpha=1}
  LinearInterpolate(A=NamedRerouteUsage(), B=@n3, Alpha=NamedRerouteUsage()) {constA=0.9, constB=0.1}
  n4 = LinearInterpolate(
   A=ScalarParameter() {group=UVs, parameterName=Background Depth Max, sortPriority=3}
   B=ScalarParameter() {defaultValue=0.2, group=UVs, parameterName=Background Depth Min, sortPriority=2}
  ) {constAlpha=0.75}
  LinearInterpolate(A=NamedRerouteUsage(), B=@n4, Alpha=NamedRerouteUsage()) {constA=0.9, constB=0.2}
  n5 = NamedRerouteUsage()
  n6 = Multiply(A=@n5, B=@n5)
  LinearInterpolate(Alpha=Saturate(None=Multiply(A=@n6) {constB=3})) {constA=1, constB=1.08}
  n7 = LinearInterpolate(
   A=ScalarParameter() {group=UVs, parameterName=Background Depth Max, sortPriority=3}
   B=ScalarParameter() {defaultValue=0.2, group=UVs, parameterName=Background Depth Min, sortPriority=2}
  ) {constAlpha=0.25}
  LinearInterpolate(A=NamedRerouteUsage(), B=@n7, Alpha=NamedRerouteUsage()) {constA=0.9, constB=0.3}
  n8 = LinearInterpolate(
   A=ScalarParameter() {group=UVs, parameterName=Background Depth Max, sortPriority=3}
   B=ScalarParameter() {defaultValue=0.2, group=UVs, parameterName=Background Depth Min, sortPriority=2}
  ) {constAlpha=0}
  LinearInterpolate(A=NamedRerouteUsage(), B=@n8, Alpha=NamedRerouteUsage()) {constA=0.9, constB=0.4}
  n9 = Saturate(None=NamedRerouteUsage())
  n10 = VectorParameter() {parameterName=LumenColor_Morning, defaultValue=<rgba 1 0.4276 0.0369 1>}
  n11 = VectorParameter() {parameterName=LumenColor_Day, defaultValue=<rgba 2 1.7351 1.2188 1>}
  n12 = VectorParameter() {parameterName=LumenColor_Evening, defaultValue=<rgba 1 0.59 0.3683 1>}
  n13 = VectorParameter() {parameterName=LumenColor_Night_Indirect, defaultValue=<rgba 0.0563 0.134 0.4115 1>}
  n14 = Multiply(
   A=Multiply(A=VectorParameter() {parameterName=Color 1, defaultValue=<rgba 0 0.7079 1 1>}.RGB, B=LinearInterpolate(Alpha=@n9) {constA=1, constB=2})
   B=LinearInterpolate(A=MaterialFunctionCall(Color_Morning=@n10.RGB, Color_Day=@n11.RGB, Color_Evening=@n12.RGB, Color_Night=@n13.RGB) {materialFunction=&Color_Time_of_Day}.Result)
  )
  n15 = ReflectionVectorWS()
  n16 = Constant3Vector() {constant=<rgba 1 1 -1 0>}
  n17 = Multiply(
   A=TextureSample(UVs=Multiply(A=@n15, B=@n16)) {texture=/Game/FN_PublicTurntable/TurntableAssets/Cubemap_Apollo_Day}.RGB
   B=ScalarParameter() {defaultValue=3.8276, parameterName=Fake reflection Intensity}
  )
  n18 = DotProduct(A=CameraVectorWS(), B=Transform(None=NamedRerouteUsage()))
  n19 = LinearInterpolate(
   B=Power(Base=LinearInterpolate(A=@n17) {constB=0.2}, Exp=ScalarParameter() {defaultValue=1, parameterName=Fake reflection Texture Power})
   Alpha=SphereMask(
    A=Saturate(None=@n18)
    B=ScalarParameter() {defaultValue=0.4, parameterName=Fake reflection Min}
    Radius=ScalarParameter() {defaultValue=0.4882, parameterName=Fake reflection Radius}
    Hardness=ScalarParameter() {defaultValue=-0.0503, parameterName=Fake reflection Hardness}
   ) {attenuationRadius=100, hardnessPercent=1}
  )
  Multiply(
   A=Multiply(A=Add(A=@n14, B=@n19), B=NamedRerouteUsage())
   B=VectorParameter() {parameterName=Final M Color, defaultValue=<rgba 0.0381 0.4792 0.3538 1>}.RGB
  )
  n20 = Multiply(A=LinearInterpolate(A=NamedRerouteUsage(), B=NamedRerouteUsage(), Alpha=@n8)) {constB=1.4}
  n21 = NamedRerouteUsage()
  n22 = Add(A=Multiply(A=ComponentMask(None=@n20) {r=true}) {constB=-1}, B=Multiply(A=@n21) {constB=0.003})
  n23 = Add(A=ComponentMask(None=@n20) {g=true}, B=Multiply(A=@n21) {constB=0.05})
  n24 = TextureSample(UVs=Multiply(A=AppendVector(A=@n22, B=@n23)) {constB=0.5}) {samplerType=SAMPLERTYPE_LinearGrayscale, texture=&/T_Seamist_Noise_HQ}
  n25 = LinearInterpolate(A=NamedRerouteUsage(), B=NamedRerouteUsage(), Alpha=@n7)
  n26 = Multiply(A=@n25) {constB=-1.3}
  n27 = NamedRerouteUsage()
  n28 = Add(A=ComponentMask(None=@n26) {r=true}, B=Multiply(A=@n27) {constB=-0.0045})
  n29 = Add(A=Multiply(A=ComponentMask(None=@n26) {g=true}) {constB=-1}, B=Multiply(A=@n27) {constB=-0.033})
  n30 = Add(
   A=@n24.G
   B=TextureSample(UVs=AppendVector(A=@n28, B=@n29)) {samplerType=SAMPLERTYPE_LinearGrayscale, texture=&T_Seamist_Noise_HQ}.G
  )
  n31 = Multiply(A=LinearInterpolate(A=NamedRerouteUsage(), B=NamedRerouteUsage(), Alpha=@n4)) {constB=1.2}
  n32 = ComponentMask(None=@n31) {r=true}
  n33 = NamedRerouteUsage()
  n34 = Multiply(A=@n33) {constB=0.006}
  n35 = ComponentMask(None=@n31) {g=true}
  n36 = Multiply(A=@n33) {constB=0.02}
  n37 = TextureSample(UVs=AppendVector(A=Add(A=@n32, B=@n34), B=Add(A=@n35, B=@n36))) {samplerType=SAMPLERTYPE_LinearGrayscale, texture=&T_Seamist_Noise_HQ}
  n38 = LinearInterpolate(A=NamedRerouteUsage(), B=NamedRerouteUsage(), Alpha=@n3)
  n39 = Multiply(A=@n38) {constB=1.1}
  n40 = ComponentMask(None=@n39) {r=true}
  n41 = NamedRerouteUsage()
  n42 = Multiply(A=@n41) {constB=-0.0066}
  n43 = ComponentMask(None=@n39) {g=true}
  n44 = Multiply(A=@n41) {constB=0.05}
  n45 = Multiply(A=AppendVector(A=Add(A=@n40, B=@n42), B=Add(A=@n43, B=@n44))) {constB=-1}
  n46 = Add(
   A=@n37.G
   B=TextureSample(UVs=Multiply(A=@n45) {constB=2}) {samplerType=SAMPLERTYPE_LinearGrayscale, texture=&T_Seamist_Noise_HQ}.G
  )
  n47 = Multiply(A=Multiply(A=Multiply(A=@n24.G, B=@n30), B=Multiply(A=@n37.G, B=@n46))) {constB=2.5}
  NamedRerouteDeclaration(None=OneMinus(None=Saturate(None=@n47))) {name=Noise Globs, nodeColor=<rgba 0.3125 1 0 1>}
  n48 = Divide(A=ComponentMask(None=LocalPosition().Z) {b=true}, B=ScalarParameter() {defaultValue=156.7344, parameterName=Height World})
  n49 = Multiply(A=@n48, B=@n48)
  NamedRerouteDeclaration(None=OneMinus(None=Saturate(None=@n49))) {name=Position mask, nodeColor=<rgba 1 0.8203 0 1>}
  n50 = CollectionParameter() {
   collection=&FortniteMaterialParameters
   parameterName=SunlightDirectionalVector
  }
  n51 = PixelNormalWS()
  n52 = Divide(A=Add(A=DotProduct(A=@n50, B=@n51)))
  n53 = MaterialFunctionCall(
   Color_Morning=LinearInterpolate(
    A=VectorParameter() {parameterName=LumenColor_Morning, defaultValue=<rgba 1 0.4276 0.0369 1>}.RGB
    B=VectorParameter() {parameterName=LumenColor_Morning_indrect, defaultValue=<rgba 0.2885 0.2974 0.3333 1>}.RGB
    Alpha=@n52
   )
   Color_Day=LinearInterpolate(
    A=VectorParameter() {parameterName=LumenColor_Day, defaultValue=<rgba 2 1.7351 1.2188 1>}.RGB
    B=VectorParameter() {parameterName=LumenColor_Day_Indirect, defaultValue=<rgba 0.1259 0.2277 0.4635 1>}.RGB
    Alpha=@n52
   )
   Color_Evening=LinearInterpolate(
    A=VectorParameter() {parameterName=LumenColor_Evening, defaultValue=<rgba 1 0.59 0.3683 1>}.RGB
    B=VectorParameter() {parameterName=LumenColor_Evening_Indirect, defaultValue=<rgba 0.1089 0.1498 0.3802 1>}.RGB
    Alpha=@n52
   )
   Color_Night=LinearInterpolate(
    A=VectorParameter() {parameterName=LumenColor_Night, defaultValue=<rgba 1.0729 1.2539 2 1>}.RGB
    B=VectorParameter() {parameterName=LumenColor_Night_Indirect, defaultValue=<rgba 0.0563 0.134 0.4115 1>}.RGB
    Alpha=@n52
   )
  ) {materialFunction=&Color_Time_of_Day}
  n54 = ScalarParameter() {defaultValue=0.5, parameterName=Lumen}
  NamedRerouteDeclaration(None=LinearInterpolate(A=Multiply(A=@n53.Result, B=@n54))) {name=Sun color, nodeColor=<rgba 0.3828 1 0 1>}
  n55 = Saturate(None=NamedRerouteUsage())
  n56 = NamedRerouteUsage()
  n57 = VertexColor()
  n58 = Saturate(None=MaterialFunctionCall(In=Add(A=@n57.G) {constB=0.3}, Contrast=Constant() {r=3}) {materialFunction=&CheapContrast}.Result)
  n59 = Saturate(None=Multiply(A=Multiply(A=@n56, B=@n58)) {constB=4})
  n60 = Add(A=Multiply(A=Multiply(A=@n55, B=@n59)) {constB=3}, B=Saturate(None=Add(A=@n59) {constB=-0.6}))
  n61 = ScalarParameter() {defaultValue=1.4, parameterName=Opacity}
  NamedRerouteDeclaration(None=DepthFade(Opacity=Multiply(A=@n60, B=@n61))) {name=Opacity, nodeColor=<rgba 0 1 0.7188 1>}
  MaterialExpressionObjectPositionWS_0 = ObjectPositionWS()
  n62 = Convert(None=@MaterialExpressionObjectPositionWS_0) {
   nodeName=Break Float3
   convertInputs [{
     type=Vector3
     expressionInput {
      expression=&self:MaterialExpressionObjectPositionWS_0
      inputName=None
      mask=0
      maskA=0
      maskB=0
      maskG=0
      maskR=0
      outputIndex=0
     }
    }]
   convertOutputs [{type=Scalar}, {type=Scalar}, {type=Scalar}]
  }
  n63 = Add(A=Add(A=@n62, B=@n62), B=@n62)
  n64 = ScalarParameter() {defaultValue=20, parameterName=Divide pos}
  n65 = Multiply(A=DotProduct(A=Divide(A=@n63, B=@n64), B=Constant() {r=3.1256})) {constB=5}
  NamedRerouteDeclaration(None=@n65) {name=Random Position, nodeColor=<rgba 1 0.7734 0 1>}
  NamedRerouteDeclaration(None=ScalarParameter() {group=2. Background, parameterName=Background Angular Distortion, sortPriority=4}) {name=background Distortion, nodeColor=<rgba 1 0 0.0234 1>}
  n66 = TextureCoordinate()
  n67 = Multiply(
   A=AppendVector(
    A=ScalarParameter() {defaultValue=2, group=UVs, parameterName=Background Tiling X, sortPriority=0}
    B=ScalarParameter() {defaultValue=1.5, group=UVs, parameterName=Background Tiling Y, sortPriority=1}
   )
   B=LinearInterpolate(Alpha=NamedRerouteUsage()) {constA=1.25, constB=0.75}
  )
  NamedRerouteDeclaration(None=Multiply(A=Add(A=@n66) {constB=-0.5}, B=Multiply(A=@n67) {constB=-1})) {name=UV Coords, nodeColor=<rgba 0 0.3437 1 1>}
  n68 = CameraPositionWS()
  n69 = WorldPosition()
  n70 = Transform(None=Normalize(VectorInput=Subtract(A=@n68, B=@n69.XYZ))) {transformSourceType=TRANSFORMSOURCE_World, transformType=TRANSFORM_Tangent}
  NamedRerouteDeclaration(None=Multiply(
    A=ComponentMask(None=@n70) {g=true, r=true}
    B=ScalarParameter() {defaultValue=2, group=UVs, parameterName=Background Tiling X, sortPriority=0}
   )) {name=Tangent Camera Vec, nodeColor=<rgba 0.6641 1 0 1>}
  n71 = Time()
  NamedRerouteDeclaration(None=Multiply(A=Add(A=@n65, B=@n71), B=ScalarParameter() {defaultValue=2, parameterName=Scroll Speed}) {constB=0.5}) {name=Random time, nodeColor=<rgba 0.9219 1 0 1>}
  n72 = CameraVectorWS()
  n73 = Transform(None=NamedRerouteUsage())
  n74 = SkyLightEnvMapSample(
   Direction=LinearInterpolate(A=Multiply(A=@n72) {constB=-1}, B=ReflectionVectorWS(CustomWorldNormal=@n73))
   Roughness=ScalarParameter() {parameterName=RoughnessSkyLightEnvMap}
  )
  n75 = ScalarParameter() {defaultValue=2, parameterName=CubeContrast}
  NamedRerouteDeclaration(None=Multiply(A=Power(Base=@n74, Exp=@n75), B=ScalarParameter() {defaultValue=0.2, parameterName=BrightnessEmissive})) {name=cubemap, nodeColor=<rgba 1 0.3984 0 1>}
  n76 = Multiply(A=@n38) {constB=4}
  n77 = Multiply(A=ComponentMask(None=@n76) {g=true}) {constB=-1}
  n78 = Multiply(A=Time()) {constB=0.32}
  NamedRerouteDeclaration(None=AppendVector(A=ComponentMask(None=@n76) {r=true}, B=Add(A=@n77, B=@n78))) {name=TinyBubblesCoord2, nodeColor=<rgba 0 0.7891 1 1>}
  n79 = NamedRerouteUsage()
  n80 = AppendVector(A=DDX(Value=@n79), B=DDY(Value=@n79))
  n81 = Constant() {r=1}
  MaterialExpressionMaterialFunctionCall_3 = MaterialFunctionCall!(InNormal=DeriveNormalZ(InXY=AppendVector(A=@n80, B=@n81)), Strength=ScalarParameter() {defaultValue=20, parameterName=Normal Intensity}) {materialFunction=/Game/Effects/Fort_Effects/FrontEnd/S33/GateContent/SilenceMoon_NormalStrength}
  n82 = Convert(None=@MaterialExpressionMaterialFunctionCall_3.Result) {
   nodeName=Break Float3
   convertInputs [{
     type=Vector3
     expressionInput {
      expression=&self:MaterialExpressionMaterialFunctionCall_3
      inputName=None
      mask=0
      maskA=0
      maskB=0
      maskG=0
      maskR=0
      outputIndex=0
     }
    }]
   convertOutputs [{type=Scalar}, {type=Scalar}, {type=Scalar}]
  }
  NamedRerouteDeclaration(None=MaterialFunctionCall(R=@n82, G=Multiply(A=@n82) {constB=-1}, B=@n82) {materialFunction=&AppendMany}.RGB) {name=Normal Noise, nodeColor=<rgba 0.1484 0 1 1>}
  n83 = Multiply(A=TextureCoordinate(), B=ScalarParameter() {defaultValue=2, parameterName=WPO frequency})
  n84 = ComponentMask(None=@n83) {g=true}
  n85 = Divide(A=NamedRerouteUsage(), B=ScalarParameter() {defaultValue=4, parameterName=WPO noise speed})
  n86 = Multiply(A=VertexNormalWS(), B=Sine(None=Add(A=@n84, B=@n85)))
  n87 = Add(A=TextureSample(UVs=Panner(Coordinate=@n83, Time=@n85) {speedY=0.2}) {samplerType=SAMPLERTYPE_Masks, texture=&/T_Seamist_Noisepack}.G) {constB=-0.5}
  NamedRerouteDeclaration(None=Multiply(A=Multiply(A=@n86, B=@n87), B=ScalarParameter() {defaultValue=10, parameterName=WPO Intensity}) {constB=20}) {name=Combined WPO, nodeColor=<rgba 0.1719 1 0 1>}
  n88 = Multiply(A=@n25) {constB=10}
  n89 = Multiply(A=ComponentMask(None=@n88) {g=true}) {constB=-1}
  n90 = Multiply(A=Time()) {constB=0.2}
  NamedRerouteDeclaration(None=AppendVector(A=ComponentMask(None=@n88) {r=true}, B=Add(A=@n89, B=@n90))) {name=TinyBubblesCoord1, nodeColor=<rgba 1 0.4688 0 1>}
  NamedRerouteUsage()
  NamedRerouteUsage()
  NamedRerouteUsage()
  NamedRerouteUsage()
  OneMinus(None=@n37.G)
  n91 = NamedRerouteUsage()
  n92 = TwoSidedSign()
  n93 = Transform(None=LinearInterpolate(A=@n91, B=Multiply(A=@n91) {constB=-1}, Alpha=Saturate(None=@n92)))
  n94 = Normalize(VectorInput=Subtract(A=CameraPositionWS(), B=ObjectPositionWS()))
  Saturate(None=NamedRerouteDeclaration(None=DotProduct(A=@n93, B=@n94)) {name=Normalized Fresnel, nodeColor=<rgba 0 1 0.2969 1>})
  TemporalResponsivenessOutput!(None=Constant() {r=1})
  Time()
  Time()
  Time()
  VertexNormalWS()
 ]
}
