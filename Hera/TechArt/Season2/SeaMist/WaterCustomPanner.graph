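# Material function for Season 2 SeaMist water that adds panning far-wave normal detail to an input normal.
# Two samples of a tiling normal pan on water time, using UVs blended between world XY and angle-based radial coordinates.
# Their flattened sum blends in by the far mask, floored by an up-close amount.
# Low quality, Forward and Mobile output only the flattened detail, with biome-masked intensity; object-space panner and static-switch nodes are unconnected.
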MaterialFunction /Hera/TechArt/Season2/SeaMist/WaterCustomPanner
graph {nodes [
  Constant3Vector() {constant=<rgba 0 0 1 0>}
  Constant3Vector() {constant=<rgba 0.1 0.1 1 0>}
  Constant() {r=1}
  n1 = AppendVector(
   A=ScalarParameter() {defaultValue=1, group=RadialUVs, parameterName=RadialUV_X}
   B=ScalarParameter() {defaultValue=1, group=RadialUVs, parameterName=RadialUV_Y}
  )
  n2 = ScalarParameter() {defaultValue=0.5, group=RadialUVs, parameterName=RadialUV_Center_X}
  n3 = ScalarParameter() {defaultValue=0.5, group=RadialUVs, parameterName=RadialUV_Center_Y}
  n4 = MaterialFunctionCall(Vector or UVs=Subtract(A=TextureCoordinate(), B=AppendVector(A=@n2, B=@n3))) {materialFunction=&VectorToRadialValue}
  n5 = AppendVector(A=@n4.Vector Converted to Angle, B=@n4.Vector Converted to Angle)
  Divide(A=@n1, B=@n5)
  Divide(A=LocalPosition().XY) {constB=100}
  n6 = ScalarParameter() {defaultValue=8000, group=Distance Blending, parameterName=FarWaveNormalScale}
  n7 = Multiply(A=@n6, B=ScalarParameter() {defaultValue=1, group=Distance Blending, parameterName=FarWaveNormalScaleFactor_LowDetail})
  n8 = QualitySwitch(Default=@n6, Low=@n7)
  n9 = Divide(
   A=ComponentMask(None=WorldPosition() {worldPositionShaderOffset=WPT_ExcludeAllShaderOffsets}.XYZ) {g=true, r=true}
   B=ShadingPathSwitch(Default=@n8, Deferred=@n8, Forward=@n7, Mobile=@n7)
  )
  n10 = Multiply(A=@n1, B=@n5)
  n11 = Saturate(None=ScalarParameter() {defaultValue=1, parameterName=UseCircularUvs})
  n12 = MaterialFunctionCall(
   Time=MaterialFunctionCall() {materialFunction=/Water/Materials/Functions/GetWaterTime}.WaterTime
   Speed=ScalarParameter() {defaultValue=1.6, group=Distance Blending, parameterName=FarWaveNormalSpeed}
  ) {materialFunction=&TimeWithSpeedVariable}
  n13 = ScalarParameter() {defaultValue=1, group=Distance Blending, parameterName=FarMipBias}
  n14 = TextureSampleParameter2D!(
   UVs=Panner(
    Coordinate=LinearInterpolate(A=@n9, B=@n10, Alpha=@n11)
    Time=@n12.Time
    Speed=VectorParameter() {group=Distance Blending, parameterName=Distance Normal Panner 1, defaultValue=<rgba 0.05 0.02 0 0>}.RGB
   )
   Bias=@n13
   Apply View MipBias
  ) {
   group=Distance Blending
   mipValueMode=TMVM_MipBias
   parameterName=FarWaveNormalHeight
   samplerSource=SSM_Wrap_WorldGroupSettings
   samplerType=SAMPLERTYPE_LinearColor
   texture=/Game/Global/Textures/Default/Tileable_Patterns/T_Water_TilingNormal_With_Height_02_Softened
  }
  n15 = ComponentMask(None=@n14.RGB) {g=true, r=true}
  n16 = Multiply(A=@n9) {constB=1.618}
  n17 = Add(A=@n10, B=Constant2Vector() {g=0.56, r=0.33}) {constB=2}
  n18 = TextureSampleParameter2D(
   UVs=Panner(
    Coordinate=LinearInterpolate(A=@n16, B=@n17, Alpha=@n11)
    Time=@n12.Time
    Speed=VectorParameter() {group=Distance Blending, parameterName=Distance Normal Panner 2, defaultValue=<rgba -0.01 -0.06 0 0>}.RGB
   )
   Bias=@n13
  ) {
   group=Distance Blending
   mipValueMode=TMVM_MipBias
   parameterName=FarWaveNormalHeight
   samplerSource=SSM_Wrap_WorldGroupSettings
   samplerType=SAMPLERTYPE_LinearColor
   texture=&T_Water_TilingNormal_With_Height_02_Softened
  }
  n19 = ComponentMask(None=@n18.RGB) {g=true, r=true}
  n20 = AppendVector(A=Add(A=ConstantBiasScale(None=@n15) {bias=-0.5, scale=2}, B=ConstantBiasScale(None=@n19) {bias=-0.5, scale=2}), B=Constant())
  n21 = ScalarParameter() {defaultValue=0.05, group=Distance Blending, parameterName=FarWavelNormalIntensity}
  n22 = FunctionInput() {bUsePreviewValueAsDefault=true, inputName=Additional Mask, inputType=FunctionInput_Scalar, sortPriority=3, previewValue=<1 0 0 1>}
  n23 = Multiply(
   A=FunctionInput(Preview=Constant()) {bUsePreviewValueAsDefault=true, inputName=Far Mask Override, inputType=FunctionInput_Scalar, sortPriority=2}
   B=@n22
  )
  n24 = LinearInterpolate(
   A=FunctionInput() {bUsePreviewValueAsDefault=true, inputName=Normal, previewValue=<0 0 1 1>}
   B=MaterialFunctionCall(Normal=@n20, Flatness=OneMinus(None=@n21)) {materialFunction=&FlattenNormal}.Result
   Alpha=Max(
    A=FunctionOutput(None=@n23) {outputName=Far Mask, sortPriority=3}
    B=ScalarParameter() {defaultValue=0.15, parameterName=DetailNormal Up Close Amount}
   )
  )
  n25 = ScalarParameter() {defaultValue=0.05, group=Distance Blending, parameterName=FarWavelNormalIntensity_LowDetail}
  n26 = FunctionInput() {bUsePreviewValueAsDefault=true, inputName=LowDetailBiome_Mask, inputType=FunctionInput_Scalar, sortPriority=8, previewValue=<1 0 0 1>}
  n27 = MaterialFunctionCall(
   Normal=@n20
   Flatness=LinearInterpolate(
    A=OneMinus(None=@n25)
    B=ScalarParameter() {defaultValue=0.99, group=Distance Blending, parameterName=FarWavelNormalIntensity_Biome_LowDetail}
    Alpha=Saturate(None=@n26)
   )
  ) {materialFunction=&FlattenNormal}
  n28 = QualitySwitch(Default=@n24, Low=@n27.Result, High=@n24, Medium=@n24)
  n29 = ShadingPathSwitch(Default=@n28, Deferred=@n28, Forward=@n27.Result, Mobile=@n27.Result)
  FunctionOutput(None=@n29) {outputName=Normal}
  n30 = Normalize(VectorInput=Transform(None=VectorParameter() {group=Distance Blending, parameterName=Distance ObjectSpace Panner, defaultValue=<rgba 1 0 0 0>}.RGB) {transformSourceType=TRANSFORMSOURCE_Local})
  Multiply(A=@n30, B=ScalarParameter() {defaultValue=0.08, group=Distance Blending, parameterName=Distance Object Panner_1})
  Multiply(A=@n30, B=ScalarParameter() {defaultValue=0.064, group=Distance Blending, parameterName=Distance Object Panner_2})
  Multiply(A=TextureCoordinate(), B=ScalarParameter() {defaultValue=1, group=Distance Blending, parameterName=Pan_UV_size})
  Reroute(None=@n29)
  StaticSwitchParameter() {group=RadialUVs, parameterName=UseRadialUVs}
  StaticSwitchParameter() {group=RadialUVs, parameterName=UseRadialUVs}
 ]}
