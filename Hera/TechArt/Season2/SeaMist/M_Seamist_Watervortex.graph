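# M_Seamist_Watervortex is a masked, two-sided Single Layer Water material for the SeaMist water vortex surface.
# World-position offset can optionally rotate the mesh about its pivot over time.
# Panning normals blend with depth-based edge foam, flattened by a radial center mask and layered with fading dual-texture caustics.
# Scattering, absorption and refraction respond to surface side, with cheaper quality and mobile branches.
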Material /Hera/TechArt/Season2/SeaMist/M_Seamist_Watervortex
graph {
 properties {
  bAutomaticallySetUsageInEditor=false
  bUseMaterialAttributes=true
  blendMode=BLEND_Masked
  physMaterial=/FortWater/Materials/WaterSurface/WaterBody
  refractionMethod=RM_PixelNormalOffset
  shadingModel=MSM_SingleLayerWater
  twoSided=true
  usedShadingModels=SingleLayerWater
 }
 nodes [
  n1 = Multiply(A=Time(), B=ScalarParameter() {defaultValue=0.1, group=WPO, parameterName=Bob Sine Speed})
  n2 = Add(
   A=Multiply(A=Time(), B=ScalarParameter() {defaultValue=0.005, group=WPO, parameterName=Rotation Speed})
   B=ScalarParameter() {defaultValue=-0.01, group=WPO, parameterName=Rotation Angle Min}
  )
  n3 = RotateAboutAxis(
   NormalizedRotationAxis=Constant3Vector() {constant=<rgba 0 0 1 0>}
   RotationAngle=@n2
   PivotPoint=ObjectPositionWS() {originType=CameraRelative}
   Position=WorldPosition() {worldPositionShaderOffset=WPT_CameraRelative}.XYZ
  )
  Add(
   A=LinearInterpolate(
    A=VectorParameter() {group=WPO, parameterName=BobHeightMax, defaultValue=<rgba 0 0 2 0>}.RGB
    B=VectorParameter() {group=WPO, parameterName=BobHeightMin, defaultValue=<rgba 0 0 -2 0>}.RGB
    Alpha=Sine(None=@n1)
   )
   B=@n3
  )
  Constant2Vector()
  Constant3Vector() {constant=<rgba 0 0 1 0>}
  Constant()
  DistanceFieldGradient!(World Position=WorldPosition().XYZ)
  Divide(A=NamedRerouteUsage(), B=ScalarParameter() {defaultValue=4050023424})
  LinearInterpolate(
   B=MaterialFunctionCall(Rotation Axis=Constant3Vector() {constant=<rgba 0 0 1 0>}, Rotation Angle=@n2, World Space Vertex Normals=Transform()) {materialFunction=&FixRotateAboutAxisNormals}.New Tangent Space Vertex Normal
   Alpha=Saturate(None=ScalarParameter() {group=WPO, parameterName="Use Rotation WPO? (Switch)", sliderMax=1})
  )
  n4 = ScalarParameter() {defaultValue=1, parameterName=FarMask}
  MaterialFunctionCall!(Normal, WPO, Far Mask Override=@n4, Additional Mask, LowDetailBiome_Mask) {materialFunction=/FortWater/Materials/Functions/WaterTwoTexturePanner}
  MaterialFunctionCall!(Mask, Scattering, Absorption) {materialFunction=/Water/Materials/Functions/WaterCoefficientMask}
  n5 = ScalarParameter() {group=WPO, parameterName="Use Rotation WPO? (Switch)", sliderMax=1}
  NamedRerouteDeclaration(None=LinearInterpolate(B=@n3, Alpha=Saturate(None=@n5))) {name=WPO, nodeColor=<rgba 1 0 0.7969 1>}
  n6 = CameraVectorWS()
  n7 = ComponentMask(None=Add(A=WorldPosition().XYZ, B=Multiply(A=@n6) {constB=-1})) {g=true, r=true}
  n8 = ScalarParameter() {defaultValue=0.002, group=Mobile Water, parameterName=Mobile Caustics1 Scale}
  n9 = Time()
  n10 = ScalarParameter() {defaultValue=0.04, group=Mobile Water, parameterName=Mobile Caustics1 Speed}
  n11 = TextureSample(UVs=Add(A=Multiply(A=@n7, B=@n8) {constB=0.0008}, B=Multiply(A=@n9, B=@n10) {constB=0.03})) {
   samplerSource=SSM_Wrap_WorldGroupSettings
   samplerType=SAMPLERTYPE_LinearGrayscale
   texture=/FortWater/Materials/WaterSurface/Mobile/Textures/T_MobileWater_Caustics
  }
  n12 = AppendVector(A=ComponentMask(None=@n7) {g=true}, B=ComponentMask(None=@n7) {r=true})
  n13 = Multiply(A=ScalarParameter() {defaultValue=0.0009, group=Mobile Water, parameterName=Mobile Caustics2 Scale}, B=Constant2Vector() {g=1, r=0.5})
  n14 = Time()
  n15 = ScalarParameter() {defaultValue=0.05, group=Mobile Water, parameterName=Mobile Caustics2 Speed}
  n16 = TextureSample(UVs=Subtract(A=Multiply(A=@n12, B=@n13) {constB=0.0007}, B=Multiply(A=@n14, B=@n15) {constB=0.02})) {
   samplerSource=SSM_Wrap_WorldGroupSettings
   samplerType=SAMPLERTYPE_LinearGrayscale
   texture=&T_MobileWater_Caustics
  }
  n17 = Multiply(A=@n11.R, B=@n16.R)
  n18 = Saturate(None=Divide(A=PixelDepth().R, B=ScalarParameter() {defaultValue=30000, group=Mobile Water, parameterName=Mobile Caustics Fade Distance}) {constB=10000})
  n19 = ScalarParameter() {defaultValue=0.6, group=Mobile Water, parameterName=Caustics Intensity}
  n20 = ScalarParameter() {defaultValue=1, group=Mobile Water, parameterName=Mobile Caustics Intensity}
  n21 = Add(A=LocalPosition().Z, B=ScalarParameter() {defaultValue=-1.5, parameterName=Caustic Height})
  n22 = ScalarParameter() {defaultValue=4.5, parameterName=Caustic Contrast}
  n23 = Multiply(A=Multiply(A=OneMinus(None=@n18), B=ShadingPathSwitch(Default=@n19, Mobile=@n20)), B=Saturate(None=Divide(A=@n21, B=@n22)))
  n24 = MaterialFunctionCall() {materialFunction=&Color_Time_of_Day}
  NamedRerouteDeclaration(None=Multiply(A=Multiply(A=@n17, B=@n23), B=LinearInterpolate(A=@n24.Result) {constAlpha=0.2})) {name=Caustics, nodeColor=<rgba 1 0.9141 0 1>}
  n25 = Subtract(A=PixelDepth().R) {constB=5000}
  NamedRerouteDeclaration(None=Saturate(None=Divide(A=@n25) {constB=10000})) {name=Pixel Depth, nodeColor=<rgba 0.2891 1 0 1>}
  n26 = Subtract(A=SceneDepthWithoutWater!(UVs).R, B=PixelDepth().R)
  n27 = ComponentMask(None=CameraVectorWS()) {b=true}
  n28 = Power(
   Base=Multiply(A=NamedRerouteDeclaration(None=@n26) {name=Scen Dpeth Water, nodeColor=<rgba 0.1953 1 0 1>}, B=Abs(None=@n27))
   Exp=ScalarParameter() {defaultValue=1.5, group=Foam, parameterName=Foam Mask contrast}
  )
  n29 = Saturate(None=Divide(A=@n28, B=ScalarParameter() {defaultValue=30, group=Foam, parameterName=Foam Range}) {constB=44})
  n30 = Multiply(A=VertexColor().R) {constB=0.75}
  n31 = Saturate(None=ScalarParameter() {group=Foam, parameterName=UseVertexColors?})
  n32 = NamedRerouteDeclaration(None=Add(A=OneMinus(None=@n29), B=LinearInterpolate(B=@n30, Alpha=@n31) {constB=0})) {name=Edge Mask, nodeColor=<rgba 1 0.6699 0 1>}
  n33 = OneMinus(None=Saturate(None=Divide(A=@n28) {constB=15}))
  n34 = OneMinus(None=Multiply(A=@n33, B=@n33))
  n35 = NamedRerouteDeclaration(None=Multiply(A=@n34, B=@n34)) {name=Edge Fade, nodeColor=<rgba 0.0781 0 1 1>}
  n36 = ComponentMask(None=NamedRerouteUsage()) {g=true, r=true}
  n37 = Add(
   A=Divide(A=WorldPosition().XY, B=ScalarParameter() {defaultValue=300, group=Foam, parameterName=Foam Scale}) {constB=800}
   B=Multiply(A=Add(A=@n36) {constB=-0.5}) {constB=0.1}
  )
  n38 = ComponentMask(None=Multiply(A=NamedRerouteUsage()) {constB=1.5}) {g=true, r=true}
  n39 = TextureSample(UVs=NamedRerouteDeclaration(None=Add(A=@n37, B=@n38)) {name=UVs, nodeColor=<rgba 1 0 0.1641 1>}) {samplerSource=SSM_Wrap_WorldGroupSettings, samplerType=SAMPLERTYPE_LinearColor, texture=&/T_Water_Hermes_Foam_M}
  n40 = Add(
   A=OneMinus(None=@n39.R)
   B=Multiply(A=NamedRerouteUsage(), B=ScalarParameter() {defaultValue=0.2, group=Foam, parameterName=Mask Offset}) {constB=0.15}
  )
  n41 = NamedRerouteUsage()
  n42 = Saturate(None=Subtract(A=Multiply(A=@n32, B=@n35), B=Multiply(A=@n40, B=@n41)))
  NamedRerouteDeclaration(None=ShadingPathSwitch(Default=Multiply(A=@n42, B=@n42), Mobile=@n42) {desc=Migrated from FeatureLevelSwitch}) {name=Foam Mask, nodeColor=<rgba 1 0 0.4219 1>}
  n43 = SetMaterialAttributes!(
   MaterialAttributes
   Base Color=Constant()
   Specular=ScalarParameter() {defaultValue=0.7, parameterName=Water Specular}
   Roughness=ScalarParameter() {defaultValue=0.1, parameterName=Water Roughness}
   Normal
   Opacity=Constant() {r=0.7}
  ) {attributeSetTypes [
    #1
    #2
    #3
    #4
    #5
   ]}
  NamedRerouteDeclaration(None=@n43) {name=Far Material, nodeColor=<rgba 1 0 0.6562 1>}
  n44 = Divide(A=WorldPosition().XY, B=ScalarParameter() {defaultValue=5000, parameterName=Offset Scale}) {constB=800}
  n45 = Time()
  n46 = Add(A=@n44, B=Multiply(A=@n45, B=Constant2Vector() {g=0.012, r=0.01}) {constB=0.01})
  n47 = Constant2Vector() {g=-0.008, r=-0.01}
  n48 = TextureSample(UVs=Add(A=Multiply(A=@n44) {constB=0.83}, B=Multiply(A=@n45, B=@n47) {constB=0.01})) {
   samplerSource=SSM_Wrap_WorldGroupSettings
   samplerType=SAMPLERTYPE_LinearColor
   texture=&/T_Water_Hermes_Bugs_Offset
  }
  n49 = Multiply(A=ComponentMask(None=@n48.RGBA) {a=true, b=true}, B=ScalarParameter() {defaultValue=0.08, parameterName=Offset Distortion}) {constB=0.1}
  n50 = TextureSample(UVs=QualitySwitch(Default=@n46, Low=@n46, High=@n46, Medium=@n46, Epic=Add(A=@n46, B=@n49))) {
   samplerSource=SSM_Wrap_WorldGroupSettings
   samplerType=SAMPLERTYPE_LinearColor
   texture=&T_Water_Hermes_Bugs_Offset
  }
  n51 = NamedRerouteDeclaration(None=@n50.RGBA) {name=Offset Texture, nodeColor=<rgba 1 0 0.8438 1>}
  n52 = ScalarParameter() {defaultValue=1.7, parameterName=Offset Pattern intensity}
  n53 = Time()
  n54 = VectorParameter() {parameterName=Offset Speed, defaultValue=<rgba 0.1 0.25 0.45 0.55>}
  n55 = AppendVector(A=@n54.RGB, B=@n54.RGB)
  MaterialExpressionSine_1 = Sine(None=Add(A=Multiply(A=@n51, B=@n52) {constB=2}, B=Multiply(A=@n53, B=@n55) {constB=0.2}))
  n56 = Convert(None=@MaterialExpressionSine_1) {
   nodeName=Break Float3
   convertInputs [{
     type=Vector3
     expressionInput {
      expression=&self:MaterialExpressionSine_1
      inputName=None
      mask=0
      maskA=0
      maskB=0
      maskG=0
      maskR=0
      outputIndex=0
     }
    }]
   convertOutputs [{type=Scalar}, {type=Scalar}]
  }
  NamedRerouteDeclaration(None=Add(A=@n56, B=@n56)) {name=Offset Slow Pattern, nodeColor=<rgba 1 0.4219 0 1>}
  NamedRerouteDeclaration(None=Add(A=ComponentMask(None=@MaterialExpressionSine_1) {a=true, b=true}, B=ComponentMask(None=@MaterialExpressionSine_1) {r=true})) {name=Offset Normal Pattern, nodeColor=<rgba 0 1 0.4609 1>}
  n57 = DeriveNormalZ(InXY=Multiply(A=NamedRerouteUsage(), B=ScalarParameter() {defaultValue=0.01, parameterName=Normal Intensity}) {constB=0.05})
  n58 = LinearInterpolate(
   A=ScalarParameter() {defaultValue=0.5, parameterName=Refraction Intensity Underwater}
   B=ScalarParameter() {defaultValue=5, parameterName=Refraction Intensity}
   Alpha=Saturate(None=TwoSidedSign())
  ) {constA=1}
  NamedRerouteDeclaration(None=Add(A=Multiply(A=@n57, B=@n58) {constB=0.3})) {name=Refraction, nodeColor=<rgba 0.4297 1 0 1>}
  n59 = VectorParameter() {group=Foam, parameterName=Foam Color, defaultValue=<rgba 0.5082 0.7448 0.7448 1>}
  n60 = ScalarParameter() {defaultValue=0.5, group=Foam, parameterName=EyeAdaptationInverse}
  n61 = TextureSample(UVs=NamedRerouteUsage()) {samplerSource=SSM_Wrap_WorldGroupSettings, samplerType=SAMPLERTYPE_Normal, texture=&/T_Water_Hermes_Foam_N}
  n62 = Constant3Vector() {constant=<rgba 0 0 1 0>}
  n63 = Saturate(None=ScalarParameter() {group=Foam, parameterName=Foam Normal Flatten, sliderMin=1})
  n64 = SetMaterialAttributes(
   Base Color=NamedRerouteDeclaration(None=EyeAdaptationInverse(LightValueInput=@n59.RGB, AlphaInput=@n60).EyeAdaptationInverse) {name=Foam Color, nodeColor=<rgba 1 0.2344 0 1>}
   Specular=Multiply(A=ScalarParameter() {defaultValue=0.4, group=Foam, parameterName=Foam Specular}, B=NamedRerouteUsage())
   Roughness=ScalarParameter() {defaultValue=1, group=Foam, parameterName=Foam Roughness}
   Normal=NamedRerouteDeclaration(None=LinearInterpolate(A=@n61.RGB, B=@n62, Alpha=@n63)) {name=Foam Normals, nodeColor=<rgba 0.9453 1 0 1>}
   Opacity=Constant() {r=0.7}
  ) {attributeSetTypes [
    #1
    #2
    #3
    #4
    #5
   ]}
  NamedRerouteDeclaration(None=@n64) {name=Foam Material, nodeColor=<rgba 0.3125 1 0 1>}
  n65 = SetMaterialAttributes!(
   MaterialAttributes
   Base Color=Constant()
   Specular=Multiply(A=ScalarParameter() {defaultValue=0.7, parameterName=Water Specular}, B=NamedRerouteUsage())
   Roughness=ScalarParameter() {defaultValue=0.1, parameterName=Water Roughness}
   Normal=NamedRerouteDeclaration(None=MaterialFunctionCall!(Normal, Far Mask Override=@n4, Additional Mask, LowDetailBiome_Mask) {materialFunction=&/WaterCustomPanner}.Normal) {name=Water MF Normal, nodeColor=<rgba 0.0626 0.4736 1 1>}
   Opacity=Constant()
   World Position Offset
  ) {attributeSetTypes [
    #1
    #2
    #3
    #4
    #5
    #6
   ]}
  NamedRerouteDeclaration(None=@n65) {name=Water material, nodeColor=<rgba 1 0 0.3984 1>}
  NamedRerouteUsage()
  QualitySwitch(Default=@n62, Low=@n62, High=@n61.RGB, Medium=@n62, Epic=@n61.RGB)
  SetMaterialAttributes!(MaterialAttributes, World Position Offset, Refraction, Opacity Mask, Emissive Color, Normal) {attributeSetTypes [
    #6
    #7
    #8
    #9
    #4
   ]}
  n66 = GetMaterialAttributes(None=BlendMaterialAttributes(A=@n65, B=@n64, Alpha=NamedRerouteUsage())) {attributeGetTypes [#2]}
  n67 = ScalarParameter() {defaultValue=0.2, parameterName=Specular Low}
  n68 = ScalarParameter() {defaultValue=0.5, parameterName=Specular Medium}
  n69 = SetMaterialAttributes!(
   MaterialAttributes=@n66.MaterialAttributes
   Specular=QualitySwitch(
    Default=@n66.Specular
    Low=Multiply(A=@n66.Specular, B=@n67) {constB=0.5}
    High=@n66.Specular
    Medium=Multiply(A=@n66.Specular, B=@n68) {constB=0.8}
    Epic=@n66.Specular
   )
  ) {attributeSetTypes [#2]}
  n70 = NamedRerouteUsage()
  n71 = Constant3Vector() {constant=<rgba 0 0 1 0>}
  n72 = Distance(A=Divide(A=LocalPosition().XY) {constB=125}, B=Constant())
  n73 = LinearInterpolate(B=Multiply(A=OneMinus(None=@n72)) {constB=3}, Alpha=ScalarParameter() {defaultValue=1, parameterName=UseCenterMask?, sliderMax=1})
  n74 = NamedRerouteDeclaration(None=LinearInterpolate(A=GetMaterialAttributes(None=@n69) {attributeGetTypes [#4]}.Normal, B=@n71, Alpha=Saturate(None=@n73))) {name=Final Normal, nodeColor=<rgba 0.5234 0 1 1>}
  n75 = Multiply(A=NamedRerouteUsage(), B=NamedRerouteUsage())
  SetMaterialAttributes!(
   MaterialAttributes=@n69
   World Position Offset=NamedRerouteUsage()
   Refraction=NamedRerouteUsage()
   Opacity Mask=Constant() {r=1}
   Emissive Color=Multiply(A=NamedRerouteUsage(), B=OneMinus(None=@n70)) {constB=100}
   Normal=ShadingPathSwitch(Default=LinearInterpolate(A=@n74, B=@n71, Alpha=@n75), Mobile=@n71)
  ) {attributeSetTypes [
    #6
    #7
    #8
    #9
    #4
   ]}
  n76 = Saturate(None=Add(A=TwoSidedSign(), B=ScalarParameter() {parameterName=Min Scattering Absorption}))
  n77 = VectorParameter() {group=Water Shading Model, parameterName=Absorption, defaultValue=<rgba 0.5 4 10 100>}
  n78 = Multiply(A=@n77.RGB, B=@n77.RGB)
  n79 = ShadingPathSwitch(Default=Constant() {r=1}, Mobile=ScalarParameter() {defaultValue=2, parameterName=Mobile Absorption Boost, sliderMin=1}) {desc=Migrated from FeatureLevelSwitch}
  n80 = Multiply(A=@n76, B=Divide(B=Divide(A=@n78, B=@n79)))
  n81 = OneMinus(None=DotProduct(A=NamedRerouteUsage(), B=Constant3Vector() {constant=<rgba 1 1 0 0>}))
  n82 = VectorParameter() {group=Water Shading Model, parameterName=Scattering, defaultValue=<rgba 0.1 0.1 0.1 0.5>}
  n83 = Divide(A=@n82.A) {constB=100}
  n84 = Constant() {r=1}
  n85 = VectorParameter() {parameterName=MobileScatterBoost, defaultValue=<rgba 1 3 5 1.5>}
  n86 = Multiply(A=@n85.RGB, B=@n85.RGB)
  n87 = Multiply(A=@n76, B=Multiply(A=Multiply(A=@n82.RGB, B=@n83), B=ShadingPathSwitch(Default=@n84, Mobile=@n86) {desc=Migrated from FeatureLevelSwitch}))
  SingleLayerWaterMaterialOutput!(
   ScatteringCoefficients=EyeAdaptationInverse(
    LightValueInput=NamedRerouteDeclaration(None=@n87) {name=ScatteringCoef, nodeColor=<rgba 0.8665 0.5177 1 1>}
    AlphaInput=ScalarParameter() {defaultValue=1, group=Water Shading Model, parameterName=EyeAdaptionBoost}
   ).EyeAdaptationInverse
   AbsorptionCoefficients=Multiply(A=NamedRerouteDeclaration(None=@n80) {name=AbsorptionCoef, nodeColor=<rgba 0.4297 1 0 1>}, B=Power(Base=@n81) {constExponent=3})
   PhaseG=ScalarParameter() {defaultValue=0.3, group=Water Shading Model, parameterName=Anisotropy}
   ColorScaleBehindWater
  )
 ]
}
